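# Sound cue for the Snow invulnerable sound in the Dopey event, under Athena gameplay cue sounds.
# A single wave player feeds Snow_Invulnerable_01 straight to the output, with no randomization or mixing.
# It routes to the Fort_Events_Scripted sound class at reduced volume, with a very large max distance.
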SoundCue /FNGameplayCues/Athena/Sounds/Events/Dopey/Snow/Snow_Invulnerable_Cue
root=WavePlayer_6
nodes [{id=WavePlayer_6, properties {soundWaveAssetPtr=&/Snow_Invulnerable_01}}]
properties {maxDistance=2097152, soundClassObject=/Game/Sounds/FortSoundClassesAndModes/Fort_Events_Scripted, volumeMultiplier=0.7018}
