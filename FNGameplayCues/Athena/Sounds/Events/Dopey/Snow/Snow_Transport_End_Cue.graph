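# Sound cue for the end of a snow transport in the Dopey event, played through the scripted events sound class.
# A random node picks with equal weight among three reused parachute skydive loop-enter wind waves.
# A modulator pitches the result down to a fixed lower pitch, and the cue sets its own attenuation override.
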SoundCue /FNGameplayCues/Athena/Sounds/Events/Dopey/Snow/Snow_Transport_End_Cue
root=Modulator_1
nodes [
 {id=Modulator_1, children [Random_0], properties {pitchMax=0.6, pitchMin=0.6}}
 {id=Random_0, children [WavePlayer_3,WavePlayer_4,WavePlayer_5], properties {weights [1,1,1]}}
 {id=WavePlayer_3, properties {soundWaveAssetPtr=/Game/Athena/Sounds/Wind_Falling/sfx_parachute_skydive_loop_enter_03}}
 {id=WavePlayer_4, properties {soundWaveAssetPtr=/Game/Athena/Sounds/Wind_Falling/sfx_parachute_skydive_loop_enter_01}}
 {id=WavePlayer_5, properties {soundWaveAssetPtr=/Game/Athena/Sounds/Wind_Falling/sfx_parachute_skydive_loop_enter_02}}
]
properties {bOverrideAttenuation=true, maxDistance=4000, soundClassObject=/Game/Sounds/FortSoundClassesAndModes/Fort_Events_Scripted}
