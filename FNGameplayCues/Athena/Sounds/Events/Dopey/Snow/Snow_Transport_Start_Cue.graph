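# Sound cue in the Dopey event's Snow audio set, named for the start of a snow transport.
# A single wave player feeds the root with the Snow_Enter_01 wave, with no randomization or layering.
# The cue routes to the SFX sound class, boosts volume, and overrides attenuation on the cue itself.
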SoundCue /FNGameplayCues/Athena/Sounds/Events/Dopey/Snow/Snow_Transport_Start_Cue
root=WavePlayer_0
nodes [{id=WavePlayer_0, properties {soundWaveAssetPtr=&/Snow_Enter_01}}]
properties {bOverrideAttenuation=true, maxDistance=4000, soundClassObject=&SFX, volumeMultiplier=1.5}
